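# Post-copy AnimBP for the M_MED_RedPepper player body cosmetic, correcting armor and belt bones on top of the attached parent's pose.
# Additive bone transforms on elbow pads, knee pads and belt blend by smoothed alphas, computed each update from limb joint twist deltas.
# A left-shoulderpad Control Rig runs last, with Is_fly fed by the inverse of bIsSkyDiving.
# The crouch input and many aim/shoulderpad variables are never set in these graphs.
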AnimBlueprint /BRCosmetics/Characters/Player/Male/Medium/Bodies/M_MED_RedPepper/Meshes/M_MED_RedPepper_AnimBP
target_skeleton=/Game/Characters/Player/Male/Medium/Bodies/M_MED_RedPepper/Meshes/M_MED_RedPepper_Skeleton
graphs [
 AnimGraph {nodes [
   n1 = "Transform (Modify) Bone - Bone: extra_elbowpad_l"(
    ComponentPose=Local To Component(LocalPose=Copy Pose From Mesh() {bUseAttachedParent=true})
    Alpha=Get Alpha_elbowpad_l() -> (Alpha_elbowpad_l:double)
    Translation="1.809,1.054,0.003"
    Rotation="0.245,-43.187,-0.107"
    Scale="1,1,1"
   ) {
    rotationMode=BMM_Additive
    rotationSpace=BCS_ParentBoneSpace
    translationMode=BMM_Additive
    translationSpace=BCS_ParentBoneSpace
    alphaBoolBlend {blendOption=Linear}
    alphaScaleBiasClamp {bInterpResult=true, interpSpeedDecreasing=12, interpSpeedIncreasing=12}
    boneToModify {boneName=extra_elbowpad_l}
    rotation=<rot 0.245 -43.187 -0.107>
    translation=<1.8087 1.0538 0.003>
   }
   n2 = "Transform (Modify) Bone - Bone: extra_elbowpad_r"(
    ComponentPose=@n1
    Alpha=Get Alpha_elbowpad_r() -> (Alpha_elbowpad_r:double)
    Translation="-2.167,-0.707,0.022"
    Rotation="0.265,-25.297,-0.028"
    Scale="1,1,1"
   ) {
    rotationMode=BMM_Additive
    rotationSpace=BCS_ParentBoneSpace
    translationMode=BMM_Additive
    translationSpace=BCS_ParentBoneSpace
    alphaBoolBlend {blendOption=Linear}
    alphaScaleBiasClamp {bInterpResult=true, interpSpeedDecreasing=12, interpSpeedIncreasing=12}
    boneToModify {boneName=extra_elbowpad_r}
    rotation=<rot 0.265 -25.2969 -0.0281>
    translation=<-2.167 -0.707 0.022>
   }
   n3 = "Transform (Modify) Bone - Bone: extra_kneepad_l"(
    ComponentPose=@n2
    Alpha=Get Alpha_kneepad_l() -> (Alpha_kneepad_l:double)
    Translation="-2.829,-2.216,-0.009"
    Rotation="0.226,-51.938,-0.138"
    Scale="1,1,1"
   ) {
    rotationMode=BMM_Additive
    rotationSpace=BCS_ParentBoneSpace
    translationMode=BMM_Additive
    translationSpace=BCS_ParentBoneSpace
    alphaBoolBlend {blendOption=Linear}
    alphaScaleBiasClamp {bInterpResult=true, interpSpeedDecreasing=12, interpSpeedIncreasing=12}
    boneToModify {boneName=extra_kneepad_l}
    rotation=<rot 0.226 -51.938 -0.138>
    translation=<-2.829 -2.216 -0.009>
   }
   n4 = "Transform (Modify) Bone - Bone: extra_kneepad_r"(
    ComponentPose=@n3
    Alpha=Get Alpha_kneepad_r() -> (Alpha_kneepad_r:double)
    Translation="2.944,2.435,0.008"
    Rotation="0.226,-51.938,-0.138"
    Scale="1,1,1"
   ) {
    rotationMode=BMM_Additive
    rotationSpace=BCS_ParentBoneSpace
    translationMode=BMM_Additive
    translationSpace=BCS_ParentBoneSpace
    alphaBoolBlend {blendOption=Linear}
    alphaScaleBiasClamp {bInterpResult=true, interpSpeedDecreasing=12, interpSpeedIncreasing=12}
    boneToModify {boneName=extra_kneepad_r}
    rotation=<rot 0.226 -51.938 -0.138>
    translation=<2.944 2.435 0.008>
   }
   n5 = "Transform (Modify) Bone - Bone: extra_belt_l"(
    ComponentPose=@n4
    Alpha=Get Alpha_belt_l() -> (Alpha_belt_l:double)
    Translation="-2.800,-2.938,-1.485"
    Rotation="34.175,-83.461,-5.397"
    Scale="1,1,1"
   ) {
    rotationMode=BMM_Additive
    rotationSpace=BCS_ParentBoneSpace
    translationMode=BMM_Additive
    translationSpace=BCS_ParentBoneSpace
    alphaBoolBlend {blendOption=Linear}
    alphaScaleBiasClamp {bInterpResult=true, interpSpeedDecreasing=12, interpSpeedIncreasing=12}
    boneToModify {boneName=extra_belt_l}
    rotation=<rot 34.175 -83.461 -5.397>
    translation=<-2.8 -2.938 -1.485>
   }
   n6 = "Transform (Modify) Bone - Bone: extra_belt_r"(
    ComponentPose=@n5
    Alpha=Get Alpha_belt_r() -> (Alpha_belt_r:double)
    Translation="-2.829,-2.216,-0.009"
    Rotation="-16.740,-69.609,-4.385"
    Scale="1,1,1"
   ) {
    rotationMode=BMM_Additive
    rotationSpace=BCS_ParentBoneSpace
    translationMode=BMM_Additive
    translationSpace=BCS_ParentBoneSpace
    alphaBoolBlend {blendOption=Linear}
    alphaScaleBiasClamp {bInterpResult=true, interpSpeedDecreasing=12, interpSpeedIncreasing=12}
    boneToModify {boneName=extra_belt_r}
    rotation=<rot -16.74 -69.609 -4.385>
    translation=<-2.829 -2.216 -0.009>
   }
   n7 = Control Rig!(
    Source=Component To Local(ComponentPose=@n6)
    Alpha
    L_Alpha:double=0
    L_trs:Transform="0,0,0|0,0,0|1,1,1"
    Is_fly:double=Get Ctrl_rig_is_fly() -> (Ctrl_rig_is_fly:double)
    Is_crouch:double=Get Ctrl_rig_is_crouch() -> (Ctrl_rig_is_crouch:double)
   ) -> (Pose) {
    alphaBoolBlend {blendOption=Linear}
    alphaScaleBiasClamp {interpSpeedDecreasing=3, interpSpeedIncreasing=3}
    controlRigAssetReference {blueprintRigClass=&/M_MED_RedPepper_CtrlRig_Shoulderpad_l.M_MED_RedPepper_CtrlRig_Shoulderpad_l_C}
   }
   Output Pose(Result=@n7) {layerGroup=None}
  ]}
 EventGraph {nodes [
   Event BlueprintUpdateAnimation()
   n1 = Get bIsSkyDiving()
   > Set Ctrl_rig_is_fly(Ctrl_rig_is_fly:double="To Float (Boolean)"(InBool=NOT Boolean(A=@n1))) -> (Output_Get:double)
   n2 = GetOwnerSkeletalMeshComponent()
   > Set Alpha_elbowpad_r(Alpha_elbowpad_r:double=GetAlphaFromReferenceTransformDelta(
     SkeletalMeshComponent=@n2
     BoneName=upperarm_r
     BaseName=lowerarm_r
     DeltaComponentType=QuaternionTwist
     InputRange=(X=30,Y=60)
     TwistAxis="0,0,1"
    )) -> (Output_Get:double)
   > Set Alpha_elbowpad_l(Alpha_elbowpad_l:double=GetAlphaFromReferenceTransformDelta(
     SkeletalMeshComponent=@n2
     BoneName=upperarm_l
     BaseName=lowerarm_l
     DeltaComponentType=QuaternionTwist
     InputRange=(X=30,Y=60)
     TwistAxis="0,0,1"
    )) -> (Output_Get:double)
   > Set Alpha_kneepad_l(Alpha_kneepad_l:double=GetAlphaFromReferenceTransformDelta(SkeletalMeshComponent=@n2, BoneName=thigh_l, BaseName=calf_l, DeltaComponentType=QuaternionTwist, InputRange=(X=30,Y=100), TwistAxis="0,0,1")) -> (Output_Get:double)
   > Set Alpha_kneepad_r(Alpha_kneepad_r:double=GetAlphaFromReferenceTransformDelta(SkeletalMeshComponent=@n2, BoneName=thigh_r, BaseName=calf_r, DeltaComponentType=QuaternionTwist, InputRange=(X=30,Y=100), TwistAxis="0,0,1")) -> (Output_Get:double)
   > Set Alpha_belt_l(Alpha_belt_l:double=GetAlphaFromReferenceTransformDelta(
     SkeletalMeshComponent=@n2
     BoneName=pelvis
     BaseName=thigh_l
     DeltaComponentType=QuaternionTwist
     InputRange=(X=-30,Y=-100)
     TwistAxis="0,0,1"
    )) -> (Output_Get:double)
   > Set Alpha_belt_r(Alpha_belt_r:double=GetAlphaFromReferenceTransformDelta(
     SkeletalMeshComponent=@n2
     BoneName=pelvis
     BaseName=thigh_r
     DeltaComponentType=QuaternionTwist
     InputRange=(X=-30,Y=-100)
     TwistAxis="0,0,1"
    )) -> (Output_Get:double)
  ]}
]
variables [
 Aim_All_Weapons {type=number}
 Aim_All_Weapons_Crouch {type=number}
 Aim_down_pistol {type=number}
 Aim_mid_pistol {type=number}
 Aim_relax {type=number}
 Aim_relax_pistol {type=number}
 Aim_up_pistol {type=number}
 Alpha_belt_l {type=number}
 Alpha_belt_r {type=number}
 Alpha_elbowpad_l {type=number}
 Alpha_elbowpad_r {type=number}
 Alpha_ExtraBelt_IsCrouch {type=number}
 Alpha_ExtraBelt_IsOther {type=number}
 Alpha_head_rot {type=number}
 Alpha_kneepad_l {type=number}
 Alpha_kneepad_r {type=number}
 Alpha_Shoulderpads_Crouch {type=number}
 Alpha_Shoulderpads_Crouch_Move {type=number}
 Alpha_Shoulderpads_Emote {type=number}
 Alpha_Shoulderpads_FrontEnd {type=number}
 Alpha_Shoulderpads_FrontEnd_Idle {type=number}
 Alpha_Shoulderpads_FrontEnd_Idle_L {type=number}
 Alpha_Shoulderpads_Idle {type=number}
 Alpha_Shoulderpads_Jump {type=number}
 Alpha_Shoulderpads_ParachuteOpened {type=number}
 Alpha_Shoulderpads_Skydiving {type=number}
 Alpha_Shoulderpads_Targeting {type=number}
 Angle_head_spine_Z {type=number}
 Ctrl_rig_is_crouch {type=number}
 Ctrl_rig_is_fly {type=number}
 HeadRot {type=Rotator}
 HearRot_Percentage {type=number, default=-1.1}
 Inv_Aim_All_Weapons_Crouch {type=number}
 Inv_Alpha_Shoulderpads_Skydiving {type=number}
 Inv_up_Aim_All_Weapons_Crouch {type=number}
 invUpAimAllWeap_noads_norelax {type=number}
 L_Emote_redpepper {type=number}
 L_Inv_Alpha_Shoulderpads_Emote {type=number}
 L_Inv_Alpha_Shoulderpads_Emote_0 {type=number}
 L_trs {type=Transform, default {location=<0 0 0>, rotation=<rot 0 0 50>, scale=<1 1 1>}}
 noads_norelax {type=boolean}
 Owner Skeletal {type=SkeletalMeshComponent}
 R_Emote_redpepper {type=number}
 R_Inv_Alpha_Shoulderpads_Emote {type=number}
 R_trs {type=Transform, default {location=<0 0 0>, rotation=<rot 0 0 0>, scale=<1 1 1>}}
]
