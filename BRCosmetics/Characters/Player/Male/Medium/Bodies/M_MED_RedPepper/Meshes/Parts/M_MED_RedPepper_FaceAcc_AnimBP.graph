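# Anim Blueprint for the RedPepper character's face accessory part, animating a folding gun on the dyn_gun bones.
# It copies the parent pose, layers additive bone rotations and offsets, shifts the gun base for backblings, and matches lowerarm_l to the owner's.
# A timer, set by targeting and held through jumps or falls, switches the fold pose.
# While it is full, head yaw relative to pelvis and spine_05 drives aim alphas.
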AnimBlueprint /BRCosmetics/Characters/Player/Male/Medium/Bodies/M_MED_RedPepper/Meshes/Parts/M_MED_RedPepper_FaceAcc_AnimBP
target_skeleton=&/M_MED_RedPepper_FaceAcc_Skeleton
graphs [
 AnimGraph {nodes [
   "Transform (Modify) Bone - Bone: dyn_gun_04"(Alpha=Get jumping_float() -> (jumping_float:double), Translation="0,0,0", Rotation="26.539,36.181,25.873", Scale="1,1,1") {
    rotationMode=BMM_Additive
    rotationSpace=BCS_BoneSpace
    translationSpace=BCS_BoneSpace
    alphaBoolBlend {blendOption=Linear}
    alphaScaleBiasClamp {bInterpResult=true, interpSpeedDecreasing=3, interpSpeedIncreasing=3}
    boneToModify {boneName=dyn_gun_04}
    rotation=<rot 26.539 36.181 25.873>
   }
   n1 = "Transform (Modify) Bone - Bone: dyn_gun_base"!(
    ComponentPose=Local To Component(LocalPose=Copy Pose From Mesh() {bCopyCurves=true, bUseAttachedParent=true})
    bAlphaBoolEnabled:bool=Get isWearingPack() -> (isWearingPack:bool)
    Translation="0,1,0"
    Rotation="0,0,0"
    Scale="1,1,1"
   ) -> (Pose) {
    alphaInputType=Bool
    rotationMode=BMM_Additive
    rotationSpace=BCS_BoneSpace
    translationMode=BMM_Additive
    translationSpace=BCS_BoneSpace
    alphaBoolBlend {blendOption=Linear}
    boneToModify {boneName=dyn_gun_base}
    translation {y=1}
   }
   n2 = "Transform (Modify) Bone - Bone: dyn_gun_01"(
    ComponentPose=@n1
    Alpha=Get Gun_Deploy_Alpha() -> (Gun_Deploy_Alpha:double)
    Translation="-1.865,-0.000,-0.000"
    Rotation="0,0,0"
    Scale="1,1,1"
   ) {
    rotationSpace=BCS_BoneSpace
    translationMode=BMM_Additive
    translationSpace=BCS_BoneSpace
    alphaBoolBlend {blendOption=Linear}
    boneToModify {boneName=dyn_gun_01}
    translation {x=-1.865}
   }
   n3 = "Transform (Modify) Bone - Bone: dyn_gun_02"(ComponentPose=@n2, Alpha=Get Gun_Deploy_Alpha(), Translation="0,0,0", Rotation="0,0,179", Scale="1,1,1") {
    rotationMode=BMM_Additive
    rotationSpace=BCS_BoneSpace
    translationSpace=BCS_BoneSpace
    alphaBoolBlend {blendOption=Linear}
    boneToModify {boneName=dyn_gun_02}
    rotation {roll=179}
   }
   n4 = "Transform (Modify) Bone - Bone: dyn_gun_03"(ComponentPose=@n3, Alpha=Get Gun_Deploy_Alpha(), Translation="0,0,0", Rotation="0,150,0", Scale="1,1,1") {
    rotationMode=BMM_Additive
    rotationSpace=BCS_BoneSpace
    translationSpace=BCS_BoneSpace
    alphaBoolBlend {blendOption=Linear}
    boneToModify {boneName=dyn_gun_03}
    rotation {yaw=150}
   }
   n5 = "Transform (Modify) Bone - Bone: dyn_gun_04"(ComponentPose=@n4, Alpha=Get Gun_Deploy_Alpha(), Translation="0,0,0", Rotation="-1.260,160.083,-179.566", Scale="1,1,1") {
    rotationMode=BMM_Additive
    rotationSpace=BCS_BoneSpace
    translationSpace=BCS_BoneSpace
    alphaBoolBlend {blendOption=Linear}
    boneToModify {boneName=dyn_gun_04}
    rotation=<rot -1.26 160.083 -179.566>
   }
   n6 = "Transform (Modify) Bone - Bone: dyn_gun_03"(
    ComponentPose=@n5
    Alpha=Get cannon_aim_up_alpha() -> (cannon_aim_up_alpha:double)
    Translation="0,0,0"
    Rotation="0.031,-56.389,-0.019"
    Scale="1,1,1"
   ) {
    rotationMode=BMM_Additive
    rotationSpace=BCS_BoneSpace
    translationSpace=BCS_BoneSpace
    alphaBoolBlend {blendOption=Linear}
    boneToModify {boneName=dyn_gun_03}
    rotation=<rot 0.031 -56.389 -0.019>
   }
   n7 = "Transform (Modify) Bone - Bone: dyn_gun_04"(ComponentPose=@n6, Alpha=Get cannon_aim_up_alpha(), Translation="0,0,0", Rotation="0.003,-14.237,0.006", Scale="1,1,1") {
    rotationMode=BMM_Additive
    rotationSpace=BCS_BoneSpace
    translationSpace=BCS_BoneSpace
    alphaBoolBlend {blendOption=Linear}
    boneToModify {boneName=dyn_gun_04}
    rotation=<rot 0.003 -14.237 0.006>
   }
   n8 = "Transform (Modify) Bone - Bone: dyn_gun_03"(
    ComponentPose=@n7
    Alpha=Get cannon_aim_down_alpha() -> (cannon_aim_down_alpha:double)
    Translation="0,0,0"
    Rotation="-0.004,32.913,0.000"
    Scale="1,1,1"
   ) {
    rotationMode=BMM_Additive
    rotationSpace=BCS_BoneSpace
    translationSpace=BCS_BoneSpace
    alphaBoolBlend {blendOption=Linear}
    boneToModify {boneName=dyn_gun_03}
    rotation=<rot -0.004 32.913 0>
   }
   n9 = "Transform (Modify) Bone - Bone: dyn_gun_04"(
    ComponentPose=@n8
    Alpha=Get cannon_aim_twist_alpha() -> (cannon_aim_twist_alpha:double)
    Translation="0,0,0"
    Rotation="-35.076,0.025,-0.041"
    Scale="1,1,1"
   ) {
    rotationMode=BMM_Additive
    rotationSpace=BCS_BoneSpace
    translationSpace=BCS_BoneSpace
    alphaBoolBlend {blendOption=Linear}
    boneToModify {boneName=dyn_gun_04}
    rotation=<rot -35.076 0.025 -0.041>
   }
   n10 = "Transform (Modify) Bone - Bone: dyn_gun_04"(
    ComponentPose=@n9
    Alpha=Get cannon_aim_up_twist_alpha() -> (cannon_aim_up_twist_alpha:double)
    Translation="0,0,0"
    Rotation="30,0,0"
    Scale="1,1,1"
   ) {
    rotationMode=BMM_Additive
    rotationSpace=BCS_BoneSpace
    translationSpace=BCS_BoneSpace
    alphaBoolBlend {blendOption=Linear}
    boneToModify {boneName=dyn_gun_04}
    rotation {pitch=30}
   }
   n11 = "Transform (Modify) Bone - Bone: dyn_gun_04"(
    ComponentPose=@n10
    Alpha=Get cannon_aim_down_twist_alpha() -> (cannon_aim_down_twist_alpha:double)
    Translation="0,0,0"
    Rotation="25,0,0"
    Scale="1,1,1"
   ) {
    rotationMode=BMM_Additive
    rotationSpace=BCS_BoneSpace
    translationSpace=BCS_BoneSpace
    alphaBoolBlend {blendOption=Linear}
    boneToModify {boneName=dyn_gun_04}
    rotation {pitch=25}
   }
   n12 = "Transform (Modify) Bone - Bone: dyn_gun_04"(ComponentPose=@n11, Alpha=Get correct_aim() -> (correct_aim:double), Translation="0,0,0", Rotation="-61.312,-0.136,-0.708", Scale="1,1,1") {
    rotationMode=BMM_Additive
    rotationSpace=BCS_BoneSpace
    translationSpace=BCS_BoneSpace
    alphaBoolBlend {blendOption=Linear}
    boneToModify {boneName=dyn_gun_04}
    rotation=<rot -61.312 -0.136 -0.708>
   }
   n13 = "Transform (Modify) Bone - Bone: lowerarm_l"!(
    ComponentPose=@n12
    bAlphaBoolEnabled:bool=True
    Translation=Get Blade_trs() -> (Blade_trs:Vector)
    Rotation=Get Blade_rot() -> (Blade_rot:Rotator)
    Scale="1,1,1"
   ) -> (Pose) {
    alphaInputType=Bool
    rotationMode=BMM_Replace
    translationMode=BMM_Replace
    alphaBoolBlend {blendOption=Linear}
    boneToModify {boneName=lowerarm_l}
    rotation=<rot -61.312 -0.136 -0.708>
   }
   n14 = Get targeting_aim_float() -> (targeting_aim_float:double)
   n15 = "Transform (Modify) Bone - Bone: dyn_gun_02"(ComponentPose=@n13, Alpha=@n14, Translation="0,0,0", Rotation="0,-30,0", Scale="1,1,1") {
    rotationMode=BMM_Additive
    rotationSpace=BCS_BoneSpace
    translationSpace=BCS_BoneSpace
    alphaBoolBlend {blendOption=Linear}
    alphaScaleBiasClamp {bInterpResult=true}
    boneToModify {boneName=dyn_gun_02}
    rotation {yaw=-30}
   }
   n16 = "Transform (Modify) Bone - Bone: dyn_gun_04"(ComponentPose=@n15, Alpha=@n14, Translation="0,0,0", Rotation="0.000,30.000,20.720", Scale="1,1,1") {
    rotationMode=BMM_Additive
    rotationSpace=BCS_BoneSpace
    translationSpace=BCS_BoneSpace
    alphaBoolBlend {blendOption=Linear}
    alphaScaleBiasClamp {bInterpResult=true}
    boneToModify {boneName=dyn_gun_04}
    rotation=<rot 0 30 20.72>
   }
   Output Pose(Result=Component To Local(ComponentPose=@n16)) {layerGroup=None}
  ]}
 EventGraph {nodes [
   Event BlueprintInitializeAnimation()
   > n1 = Cast To PlayerPawn_Athena(Object=TryGetPawnOwner())
   > Set Body_Mesh(Body_Mesh=GetSkeletalMeshForPartType(self=@n1))
   n2 = Event BlueprintUpdateAnimation()
   > n3 = Sequence!(execute) -> (then_0, then_1, then_2:Exec)
   > n4 = Sequence!(execute) -> (then_0, then_1, then_2, then_3:Exec)
   n5 = Get timer() -> (timer:double)
   > n6 = Branch(Condition="Equal (Float)"(A=@n5, B=1))
   > Set Gun_Deploy_Alpha(Gun_Deploy_Alpha:double=FInterpTo_Constant!(Current:double=Get Gun_Deploy_Alpha(), Target:double=0, DeltaTime:double=@n2.DeltaTimeX, InterpSpeed:double=2) -> (ReturnValue:double)) -> (Output_Get:double)
   > Set cannon_aim_twist_alpha(cannon_aim_twist_alpha:double=1) -> (Output_Get:double)
   Set Gun_Deploy_Alpha(execute=@n6.else, Gun_Deploy_Alpha=FInterpTo_Constant(Current=Get Gun_Deploy_Alpha(), Target=1, DeltaTime=@n2.DeltaTimeX, InterpSpeed=2))
   > Set cannon_aim_twist_alpha(cannon_aim_twist_alpha=0)
   n7 = "Equal (Float)"(A=Get timer(), B=1)
   n8 = Branch(execute=@n4.then_1, Condition=NOT Boolean(A=@n7))
   n9 = FInterpTo_Constant(Current=Get cannon_aim_up_alpha(), Target=1, DeltaTime=@n2.DeltaTimeX, InterpSpeed=3)
   > Set cannon_aim_up_alpha(cannon_aim_up_alpha:double=@n9) -> (Output_Get:double)
   > Set cannon_aim_up_twist_alpha(cannon_aim_up_twist_alpha:double=MapRangeClamped(Value=@n9, InRangeA=1, InRangeB=0)) -> (Output_Get:double)
   n10 = GetOwnerSkeletalMeshComponent()
   n11 = BreakRotator(InRot=BreakTransform(InTransform=GetDeltaTransformFromRefPose(self=@n10, BoneName=head, BaseName=pelvis)).Rotation)
   n12 = FInterpTo_Constant(
    Current=Get cannon_aim_up_alpha()
    Target=MapRangeClamped(Value=@n11.Yaw, InRangeA=10, InRangeB=-60)
    DeltaTime=@n2.DeltaTimeX
    InterpSpeed=3
   )
   Set cannon_aim_up_alpha(execute=@n8.else, cannon_aim_up_alpha=@n12)
   > Set cannon_aim_up_twist_alpha(cannon_aim_up_twist_alpha=MapRangeClamped(Value=@n12, InRangeA=1, InRangeB=0))
   n13 = Branch(execute=@n4.then_2, Condition=NOT Boolean(A=@n7))
   n14 = FInterpTo_Constant(Current=Get cannon_aim_down_alpha(), Target=1, DeltaTime=@n2.DeltaTimeX, InterpSpeed=3)
   > Set cannon_aim_down_alpha(cannon_aim_down_alpha:double=@n14) -> (Output_Get:double)
   > Set cannon_aim_down_twist_alpha(cannon_aim_down_twist_alpha:double=MapRangeClamped(Value=@n14, InRangeA=1, InRangeB=0)) -> (Output_Get:double)
   n15 = FInterpTo_Constant(
    Current=Get cannon_aim_down_alpha()
    Target=MapRangeClamped(Value=@n11.Yaw, InRangeA=10, InRangeB=40)
    DeltaTime=@n2.DeltaTimeX
    InterpSpeed=3
   )
   Set cannon_aim_down_alpha(execute=@n13.else, cannon_aim_down_alpha=@n15)
   > Set cannon_aim_down_twist_alpha(cannon_aim_down_twist_alpha=MapRangeClamped(Value=@n15, InRangeA=1, InRangeB=0))
   n16 = Branch(execute=@n4.then_3, Condition=Get bIsTargeting())
   > Set timer(timer:double=1) -> (Output_Get:double)
   n17 = Get bIsJumping()
   n18 = Get bIsFalling()
   n19 = Get timer()
   n20 = Branch(execute=@n16.else, Condition=AND Boolean(A=OR Boolean!(A=@n17, B=@n18, C:bool) -> (ReturnValue), B="Not Equal (Float)"(A=@n19)))
   > Set timer(timer=1)
   Set timer(execute=@n20.else, timer=FInterpTo_Constant(Current=Get timer(), Target=0, DeltaTime=@n2.DeltaTimeX, InterpSpeed=10))
   n21 = GetOwnerSkeletalMeshComponent()
   n22 = BreakRotator(InRot=BreakTransform(InTransform=GetDeltaTransformFromRefPose(self=@n21, BoneName=head, BaseName=spine_05)).Rotation)
   n23 = Get timer()
   n24 = "To Float (Boolean)"(InBool="Equal (Float)"(A=@n23, B=1))
   Set correct_aim(
    execute=@n3.then_1
    correct_aim:double=float * float!(
     A=MapRangeClamped(Value=@n22.Yaw, InRangeA=-50, InRangeB=-5)
     B=MapRangeClamped(Value=@n22.Yaw, InRangeA=5, InRangeB=20, OutRangeA=1, OutRangeB=0)
     C:double=@n24
    ) -> (ReturnValue) {is=title K2Node_PromotableOperator}
   ) -> (Output_Get:double)
   > Set targeting_aim_float(targeting_aim_float:double=@n24) -> (Output_Get:double)
   n25 = "Not Equal (Float)"(A=@n23)
   n26 = "Not Equal (Float)"(A=@n23, B=1)
   > Set jumping_float(jumping_float:double="To Float (Boolean)"(InBool=AND Boolean(A=@n26, B=@n25))) -> (Output_Get:double)
   n27 = BreakTransform(InTransform=GetSocketTransform(self=GetOwnerSkeletalMeshComponent(), InSocketName=lowerarm_l, TransformSpace=RTS_Component))
   Set Blade_trs(execute=@n3.then_2, Blade_trs:Vector=@n27) -> (Output_Get:Vector)
   > Set Blade_rot(Blade_rot:Rotator=@n27.Rotation) -> (Output_Get:Rotator)
   Event OnCharacterCustomizationLoaded()
   n28 = CheckForBackbling()
   > Set isWearingPack(isWearingPack:bool=@n28) -> (Output_Get:bool)
   > Set isWearingCape(isWearingCape=AND Boolean(A=@n28, B=@n28.bIsBackblingLong))
   "Not Equal (Float)"(A=@n5)
   Get bIsTargeting()
  ]}
]
variables [
 Blade_rot {type=Rotator}
 Blade_rot_0 {type=Rotator}
 Blade_trs {type=Vector}
 Blade_trs_0 {type=Vector}
 Body_Mesh {type=SkeletalMeshComponent}
 cannon_aim_down_alpha {type=number, default=1}
 cannon_aim_down_twist_alpha {type=number}
 cannon_aim_twist_alpha {type=number}
 cannon_aim_up_alpha {type=number, default=1}
 cannon_aim_up_twist_alpha {type=number}
 correct_aim {type=number}
 Gun_Deploy_Alpha {type=number, default=1}
 isWearingCape {type=boolean}
 isWearingPack {type=boolean}
 jumping_float {type=number}
 targeting_aim_float {type=number}
 timer {type=number}
 timer_0 {type=number}
]
